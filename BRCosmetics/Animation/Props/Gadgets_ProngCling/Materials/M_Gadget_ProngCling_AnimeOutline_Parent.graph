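# Parent material for the ProngCling gadget's anime outline, an unlit, two-sided masked material for skeletal meshes that supports portal clipping.
# An inverted-hull outline function feeds cloth wind, then the opacity mask is modified before the shared character PostFX function.
# World position relative to a portal-center parameter is transformed into a parameter-driven pivot basis.
# Its stepped Y component, combined with an inverted mask that appears to be the texture's PortalChannel, cuts the outline.
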Material /BRCosmetics/Animation/Props/Gadgets_ProngCling/Materials/M_Gadget_ProngCling_AnimeOutline_Parent
graph {
 expressions {
  MaterialExpressionAdd_7 {
   class=MaterialExpressionAdd
   inputs [
    A {from=MaterialExpressionCameraPositionWS_0}
    B {from=MaterialExpressionWorldPosition_2, output=XYZ}
   ]
  }
  MaterialExpressionCameraPositionWS_0 {class=MaterialExpressionCameraPositionWS, inputs [], properties {}}
  MaterialExpressionComponentMask_3 {
   class=MaterialExpressionComponentMask
   inputs [None {from=MaterialExpressionMaterialFunctionCall_3, output=Transformed Vector 3}]
   properties {g=true}
  }
  MaterialExpressionComposite_0 {class=MaterialExpressionComposite, inputs [], properties {subgraphName=Construct Base Inputs}}
  MaterialExpressionConstant_0 {class=MaterialExpressionConstant}
  MaterialExpressionConstant_1 {class=MaterialExpressionConstant}
  MaterialExpressionFresnel_0 {
   class=MaterialExpressionFresnel
   inputs [
    ExponentIn {from="", output=""}
    BaseReflectFractionIn {from="", output=""}
    Normal {from="", output=""}
   ]
   properties {baseReflectFraction=0, exponent=1}
  }
  MaterialExpressionGetMaterialAttributes_0 {
   class=MaterialExpressionGetMaterialAttributes
   inputs [None {from=MaterialExpressionMaterialFunctionCall_1, output=Result}]
   properties {attributeGetTypes [#1]}
  }
  MaterialExpressionMaterialFunctionCall_0 {
   class=MaterialExpressionMaterialFunctionCall
   inputs []
   properties {materialFunction=/Game/Global/Materials/Character_Parent_2023/Anime_Parent_2023/MaterialFunctions/MFC_AnimeOutline_InvHull_2023}
  }
  MaterialExpressionMaterialFunctionCall_1 {
   class=MaterialExpressionMaterialFunctionCall
   inputs [
    Material {from=MaterialExpressionMaterialFunctionCall_0, output=Result}
    FX {from=MaterialExpressionMaterialFunctionCall_0, output=Result}
   ]
   properties {materialFunction=/Game/Packages/Fortress_SharedMaterials/Anime_Master_Material/MFC_ClothWind}
  }
  MaterialExpressionMaterialFunctionCall_14 {
   class=MaterialExpressionMaterialFunctionCall
   inputs [Material {from=""}]
   properties {materialFunction=/Game/Packages/Fortress_SharedMaterials/Base_Material_Functions/CharacterBrightnessandScale_Balance}
  }
  MaterialExpressionMaterialFunctionCall_16 {
   class=MaterialExpressionMaterialFunctionCall
   inputs [
    Material {from=MaterialExpressionSetMaterialAttributes_0}
    M {from=MaterialExpressionReroute_8, output=M}
    Pre Skinned Local Position {from=MaterialExpressionReroute_9, output=Pre Skinned Local Position}
    Distance Blend {from=MaterialExpressionReroute_10, output=Distance Blend}
    Fresnel {from=MaterialExpressionReroute_11, output=Fresnel}
   ]
   properties {materialFunction=&MF_CharacterParent_2023_PostFX}
  }
  MaterialExpressionMaterialFunctionCall_3 {
   class=MaterialExpressionMaterialFunctionCall
   inputs [
    VectorToTransform {from=MaterialExpressionSubtract_5}
    BasisX {from=MaterialExpressionNamedRerouteDeclaration_11}
    BasisY {from=MaterialExpressionNamedRerouteDeclaration_12}
    BasisZ {from=MaterialExpressionNamedRerouteDeclaration_13}
   ]
   properties {materialFunction=/Engine/Functions/Engine_MaterialFunctions02/Math/InverseTransformMatrix}
  }
  MaterialExpressionMultiply_0 {
   class=MaterialExpressionMultiply
   inputs [
    A {from=MaterialExpressionGetMaterialAttributes_0, output=Opacity Mask}
    B {from=MaterialExpressionMultiply_1}
   ]
  }
  MaterialExpressionMultiply_1 {
   class=MaterialExpressionMultiply
   inputs [
    A {from=MaterialExpressionStep_1}
    B {from=MaterialExpressionOneMinus_2}
   ]
  }
  MaterialExpressionNamedRerouteDeclaration_1 {
   class=MaterialExpressionNamedRerouteDeclaration
   inputs [None {from=MaterialExpressionTextureSample_4, output=R}]
   properties {name=PortalChannel, nodeColor=<rgba 0.3359 0 1 1>}
  }
  MaterialExpressionNamedRerouteDeclaration_11 {
   class=MaterialExpressionNamedRerouteDeclaration
   inputs [None {from=MaterialExpressionVectorParameter_2, output=RGB}]
   properties {name=BasisX, nodeColor=<rgba 1 0 0.2578 1>}
  }
  MaterialExpressionNamedRerouteDeclaration_12 {
   class=MaterialExpressionNamedRerouteDeclaration
   inputs [None {from=MaterialExpressionVectorParameter_10, output=RGB}]
   properties {name=BasisY, nodeColor=<rgba 0.9922 1 0 1>}
  }
  MaterialExpressionNamedRerouteDeclaration_13 {
   class=MaterialExpressionNamedRerouteDeclaration
   inputs [None {from=MaterialExpressionVectorParameter_11, output=RGB}]
   properties {name=BasisZ, nodeColor=<rgba 0.9453 0 1 1>}
  }
  MaterialExpressionNamedRerouteUsage_3 {class=MaterialExpressionNamedRerouteUsage}
  MaterialExpressionOneMinus_2 {class=MaterialExpressionOneMinus, inputs [None {from=MaterialExpressionNamedRerouteUsage_3}]}
  MaterialExpressionPinBase_0 {class=MaterialExpressionPinBase, inputs [], properties {}}
  MaterialExpressionPinBase_1 {
   class=MaterialExpressionPinBase
   inputs [
    M {from=MaterialExpressionConstant_1, output=""}
    Pre Skinned Local Position {from=MaterialExpressionShaderStageSwitch_0, output=""}
    Distance Blend {from=MaterialExpressionMaterialFunctionCall_14, output=Distance Blend}
    Fresnel {from=MaterialExpressionFresnel_0, output=""}
   ]
   properties {reroutePins [
     M {expression=&self:MaterialGraph_1.MaterialGraphNode_Composite_0.Construct Base Inputs.MaterialExpressionReroute_8}
     Pre Skinned Local Position {expression=&self:MaterialGraph_1.MaterialGraphNode_Composite_0.Construct Base Inputs.MaterialExpressionReroute_9}
     Distance Blend {expression=&self:MaterialGraph_1.MaterialGraphNode_Composite_0.Construct Base Inputs.MaterialExpressionReroute_10}
     Fresnel {expression=&self:MaterialGraph_1.MaterialGraphNode_Composite_0.Construct Base Inputs.MaterialExpressionReroute_11}
    ]}
  }
  MaterialExpressionPreSkinnedPosition_0 {class=MaterialExpressionPreSkinnedPosition, inputs [], properties {}}
  MaterialExpressionSetMaterialAttributes_0 {
   class=MaterialExpressionSetMaterialAttributes
   inputs [
    MaterialAttributes {from=MaterialExpressionGetMaterialAttributes_0, output=MaterialAttributes}
    Opacity Mask {from=MaterialExpressionMultiply_0, output=""}
    Emissive Color {from="", output=""}
   ]
   properties {attributeSetTypes [#1, #2]}
  }
  MaterialExpressionShaderStageSwitch_0 {
   class=MaterialExpressionShaderStageSwitch
   inputs [
    PixelShader {from=MaterialExpressionVertexInterpolator_0, output=PS}
    VertexShader {from=MaterialExpressionPreSkinnedPosition_0, output=""}
   ]
   properties {}
  }
  MaterialExpressionStep_1 {
   class=MaterialExpressionStep
   inputs [
    Y {from=MaterialExpressionConstant_0, output=""}
    X {from=MaterialExpressionComponentMask_3, output=""}
   ]
   properties {constX=0}
  }
  MaterialExpressionSubtract_5 {
   class=MaterialExpressionSubtract
   inputs [
    A {from=MaterialExpressionAdd_7}
    B {from=MaterialExpressionVectorParameter_0, output=RGB}
   ]
  }
  MaterialExpressionTextureSample_4 {
   class=MaterialExpressionTextureSample
   inputs [
    UVs {from=""}
    Tex {from=""}
    Apply View MipBias {from=""}
   ]
   properties {samplerType=SAMPLERTYPE_Masks, texture=/BRCosmetics/Animation/Props/Gadgets_ProngCling/Textures/T_ProngCling_Gadgets_TA}
  }
  MaterialExpressionVectorParameter_0 {class=MaterialExpressionVectorParameter, properties {parameterName=PC_PortalCenterLoc, defaultValue=<rgba 0 0 0 1>}}
  MaterialExpressionVectorParameter_10 {class=MaterialExpressionVectorParameter, properties {group=PrismParticle, parameterName=PC_PivotBasisY, defaultValue=<rgba 0 1 0 0>}}
  MaterialExpressionVectorParameter_11 {class=MaterialExpressionVectorParameter, properties {group=PrismParticle, parameterName=PC_PivotBasisZ, defaultValue=<rgba 0 0 1 0>}}
  MaterialExpressionVectorParameter_2 {class=MaterialExpressionVectorParameter, properties {group=PrismParticle, parameterName=PC_PivotBasisX, defaultValue=<rgba 1 0 0 0>}}
  MaterialExpressionVertexInterpolator_0 {class=MaterialExpressionVertexInterpolator, inputs [VS {from=MaterialExpressionPreSkinnedPosition_0, output=""}], properties {}}
  MaterialExpressionWorldPosition_2 {class=MaterialExpressionWorldPosition, inputs [], properties {worldPositionShaderOffset=WPT_CameraRelative}}
 }
 properties {
  bAutomaticallySetUsageInEditor=false
  bEnableResponsiveAA=true
  bUseMaterialAttributes=true
  bUsedWithMeshDeformer=true
  bUsedWithSkeletalMesh=true
  blendMode=BLEND_Masked
  pixelDepthOffsetMode=PDOM_Legacy
  shadingModel=MSM_Unlit
  twoSided=true
  usedShadingModels=Unlit
 }
}
