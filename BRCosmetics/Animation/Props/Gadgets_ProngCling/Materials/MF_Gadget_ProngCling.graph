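# Material function for the ProngCling gadget that paints an animated, noise-edged portal into the lit and shaded colors and supplies an opacity mask.
# Pre-skinning local XZ becomes portal UVs, where time-animated radial Perlin samples erode a sphere mask into a stepped fill and smooth rim.
# A rotating swirl texture and a portal channel texture feed the color and masks through named reroutes.
# World position, transformed into the parameter-driven portal center and pivot basis, gives a half-space cut for opacity.
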MaterialFunction /BRCosmetics/Animation/Props/Gadgets_ProngCling/Materials/MF_Gadget_ProngCling
graph {expressions {
  MaterialExpressionAdd_0 {
   class=MaterialExpressionAdd
   inputs [
    A {from=MaterialExpressionMaterialFunctionCall_0, output=Radial Coordinates}
    B {from=MaterialExpressionFrac_0}
   ]
  }
  MaterialExpressionAdd_1 {
   class=MaterialExpressionAdd
   inputs [
    A {from=MaterialExpressionMultiply_6}
    B {from=MaterialExpressionMultiply_7}
   ]
   properties {constB=0}
  }
  MaterialExpressionAdd_2 {
   class=MaterialExpressionAdd
   inputs [
    A {from=MaterialExpressionMultiply_10}
    B {from=MaterialExpressionMultiply_9}
   ]
  }
  MaterialExpressionAdd_5 {
   class=MaterialExpressionAdd
   inputs [
    A {from=MaterialExpressionStep_0}
    B {from=MaterialExpressionSmoothStep_1}
   ]
  }
  MaterialExpressionAdd_6 {
   class=MaterialExpressionAdd
   inputs [
    A {from=MaterialExpressionNamedRerouteUsage_1}
    B {from=MaterialExpressionMultiply_8}
   ]
   properties {constB=0}
  }
  MaterialExpressionAdd_7 {
   class=MaterialExpressionAdd
   inputs [
    A {from=MaterialExpressionCameraPositionWS_0}
    B {from=MaterialExpressionWorldPosition_2, output=XYZ}
   ]
  }
  MaterialExpressionAdd_9 {
   class=MaterialExpressionAdd
   inputs [
    A {from=MaterialExpressionMultiply_13}
    B {from=MaterialExpressionSmoothStep_1}
   ]
  }
  MaterialExpressionCameraPositionWS_0 {class=MaterialExpressionCameraPositionWS, inputs [], properties {}}
  MaterialExpressionComponentMask_0 {
   class=MaterialExpressionComponentMask
   inputs [None {from=MaterialExpressionMaterialFunctionCall_0, output=Radial Coordinates}]
   properties {g=true}
  }
  MaterialExpressionComponentMask_1 {class=MaterialExpressionComponentMask, inputs [None {from=MaterialExpressionNamedRerouteDeclaration_3}], properties {b=true, r=true}}
  MaterialExpressionComponentMask_3 {
   class=MaterialExpressionComponentMask
   inputs [None {from=MaterialExpressionMaterialFunctionCall_3, output=Transformed Vector 3}]
   properties {g=true}
  }
  MaterialExpressionConstant2Vector_0 {class=MaterialExpressionConstant2Vector, properties {g=0.5, r=0.5}}
  MaterialExpressionConstant2Vector_1 {class=MaterialExpressionConstant2Vector, properties {g=5, r=1}}
  MaterialExpressionConstant2Vector_2 {class=MaterialExpressionConstant2Vector, properties {g=4, r=3}}
  MaterialExpressionConstant2Vector_3 {class=MaterialExpressionConstant2Vector, properties {g=4, r=3}}
  MaterialExpressionConstant2Vector_4 {class=MaterialExpressionConstant2Vector, properties {g=-63, r=-65}}
  MaterialExpressionConstant3Vector_0 {class=MaterialExpressionConstant3Vector, properties {constant=<rgba 0.0703 0.6717 0.75 1>}}
  MaterialExpressionConstant3Vector_1 {class=MaterialExpressionConstant3Vector, properties {constant=<rgba 1 1 1 0>}}
  MaterialExpressionConstant_0 {class=MaterialExpressionConstant}
  MaterialExpressionConstant_1 {class=MaterialExpressionConstant, properties {r=0.5}}
  MaterialExpressionConvert_1 {
   class=MaterialExpressionConvert
   inputs [
    None {from=MaterialExpressionScalarParameter_1}
    None {from=MaterialExpressionScalarParameter_2}
   ]
   properties {
    nodeName=Make Vector2
    convertInputs [
     {
      type=Scalar
      expressionInput {
       expression=&self:MaterialExpressionScalarParameter_1
       inputName=None
       mask=0
       maskA=0
       maskB=0
       maskG=0
       maskR=0
       outputIndex=0
      }
     }
     {
      type=Scalar
      expressionInput {
       expression=&self:MaterialExpressionScalarParameter_2
       inputName=None
       mask=0
       maskA=0
       maskB=0
       maskG=0
       maskR=0
       outputIndex=0
      }
     }
    ]
    convertOutputs [{type=Vector2}]
   }
  }
  MaterialExpressionDDX_0 {class=MaterialExpressionDDX, inputs [Value {from=MaterialExpressionTextureCoordinate_1, output=""}], properties {}}
  MaterialExpressionDDY_0 {class=MaterialExpressionDDY, inputs [Value {from=MaterialExpressionTextureCoordinate_1, output=""}], properties {}}
  MaterialExpressionDistance_0 {
   class=MaterialExpressionDistance
   inputs [
    A {from=MaterialExpressionSmoothStep_0, output=""}
    B {from=MaterialExpressionConstant_1, output=""}
   ]
   properties {}
  }
  MaterialExpressionDivide_0 {
   class=MaterialExpressionDivide
   inputs [
    A {from=MaterialExpressionSubtract_3}
    B {from=""}
   ]
   properties {constB=128}
  }
  MaterialExpressionFrac_0 {class=MaterialExpressionFrac, inputs [None {from=MaterialExpressionMultiply_4, output=""}], properties {}}
  MaterialExpressionFunctionInput_0 {class=MaterialExpressionFunctionInput, inputs [Preview {from="", output=""}], properties {inputName=Color_Lit_In}}
  MaterialExpressionFunctionInput_1 {class=MaterialExpressionFunctionInput, inputs [Preview {from="", output=""}], properties {inputName=Color_Shaded_In, sortPriority=1}}
  MaterialExpressionFunctionOutput_0 {
   class=MaterialExpressionFunctionOutput
   inputs [None {from=MaterialExpressionLinearInterpolate_2, output=""}]
   properties {outputName=Color_Lit}
  }
  MaterialExpressionFunctionOutput_1 {
   class=MaterialExpressionFunctionOutput
   inputs [None {from=MaterialExpressionAdd_6, output=""}]
   properties {outputName=OpacityMask, sortPriority=2}
  }
  MaterialExpressionFunctionOutput_2 {
   class=MaterialExpressionFunctionOutput
   inputs [None {from=MaterialExpressionLinearInterpolate_5, output=""}]
   properties {outputName=Color_Shaded, sortPriority=1}
  }
  MaterialExpressionLinearInterpolate_0 {
   class=MaterialExpressionLinearInterpolate
   inputs [
    A {from=""}
    B {from=MaterialExpressionConstant3Vector_0}
    Alpha {from=MaterialExpressionStep_0}
   ]
  }
  MaterialExpressionLinearInterpolate_1 {
   class=MaterialExpressionLinearInterpolate
   inputs [
    A {from=MaterialExpressionLinearInterpolate_0}
    B {from=MaterialExpressionConstant3Vector_1}
    Alpha {from=MaterialExpressionAdd_9}
   ]
  }
  MaterialExpressionLinearInterpolate_2 {
   class=MaterialExpressionLinearInterpolate
   inputs [
    A {from=MaterialExpressionFunctionInput_0}
    B {from=MaterialExpressionNamedRerouteUsage_5}
    Alpha {from=MaterialExpressionNamedRerouteUsage_0}
   ]
  }
  MaterialExpressionLinearInterpolate_5 {
   class=MaterialExpressionLinearInterpolate
   inputs [
    A {from=MaterialExpressionFunctionInput_1}
    B {from=MaterialExpressionNamedRerouteUsage_5}
    Alpha {from=MaterialExpressionNamedRerouteUsage_0}
   ]
  }
  MaterialExpressionLocalPosition_0 {class=MaterialExpressionLocalPosition, inputs [], properties {localOrigin=InstancePreSkinning}}
  MaterialExpressionMaterialFunctionCall_0 {
   class=MaterialExpressionMaterialFunctionCall
   inputs [
    Vector or UVs {from=MaterialExpressionSubtract_1}
    Swizzle Coordinate Output {from=""}
   ]
   properties {materialFunction=&VectorToRadialValue}
  }
  MaterialExpressionMaterialFunctionCall_2 {
   class=MaterialExpressionMaterialFunctionCall
   inputs [
    UVs {from=MaterialExpressionDivide_0}
    Rotation Center {from=MaterialExpressionConstant2Vector_0}
    "Rotation Angle (0-1)" {from=MaterialExpressionFrac_0}
   ]
   properties {materialFunction=&CustomRotator}
  }
  MaterialExpressionMaterialFunctionCall_3 {
   class=MaterialExpressionMaterialFunctionCall
   inputs [
    VectorToTransform {from=MaterialExpressionSubtract_5}
    BasisX {from=MaterialExpressionNamedRerouteDeclaration_11}
    BasisY {from=MaterialExpressionNamedRerouteDeclaration_12}
    BasisZ {from=MaterialExpressionNamedRerouteDeclaration_13}
   ]
   properties {materialFunction=/Engine/Functions/Engine_MaterialFunctions02/Math/InverseTransformMatrix}
  }
  MaterialExpressionMultiply_0 {
   class=MaterialExpressionMultiply
   inputs [
    A {from=MaterialExpressionAdd_0}
    B {from=MaterialExpressionConstant2Vector_1}
   ]
  }
  MaterialExpressionMultiply_1 {
   class=MaterialExpressionMultiply
   inputs [
    A {from=MaterialExpressionAdd_5}
    B {from=MaterialExpressionNamedRerouteUsage_2}
   ]
  }
  MaterialExpressionMultiply_10 {
   class=MaterialExpressionMultiply
   inputs [
    A {from=MaterialExpressionMaterialFunctionCall_0, output=Radial Coordinates}
    B {from=MaterialExpressionConstant2Vector_3}
   ]
  }
  MaterialExpressionMultiply_13 {
   class=MaterialExpressionMultiply
   inputs [
    A {from=MaterialExpressionStep_0}
    B {from=MaterialExpressionNamedRerouteUsage_4}
   ]
  }
  MaterialExpressionMultiply_2 {
   class=MaterialExpressionMultiply
   inputs [
    A {from=MaterialExpressionTextureSample_0, output=R}
    B {from=""}
   ]
   properties {constB=0.5}
  }
  MaterialExpressionMultiply_3 {
   class=MaterialExpressionMultiply
   inputs [
    A {from=MaterialExpressionTextureSample_1, output=G}
    B {from=""}
   ]
   properties {constB=0.2}
  }
  MaterialExpressionMultiply_4 {
   class=MaterialExpressionMultiply
   inputs [
    A {from=MaterialExpressionTime_0}
    B {from=MaterialExpressionConvert_1}
   ]
   properties {constB=0.5}
  }
  MaterialExpressionMultiply_5 {
   class=MaterialExpressionMultiply
   inputs [
    A {from=MaterialExpressionSphereMask_0}
    B {from=""}
   ]
   properties {constB=1.15}
  }
  MaterialExpressionMultiply_6 {
   class=MaterialExpressionMultiply
   inputs [
    A {from=MaterialExpressionMaterialFunctionCall_0, output=Radial Coordinates}
    B {from=MaterialExpressionConstant2Vector_2}
   ]
  }
  MaterialExpressionMultiply_7 {
   class=MaterialExpressionMultiply
   inputs [
    A {from=MaterialExpressionComponentMask_0}
    B {from=MaterialExpressionTextureSample_2, output=B}
   ]
   properties {constB=0.5}
  }
  MaterialExpressionMultiply_8 {
   class=MaterialExpressionMultiply
   inputs [
    A {from=MaterialExpressionOneMinus_2}
    B {from=MaterialExpressionStep_1}
   ]
  }
  MaterialExpressionMultiply_9 {
   class=MaterialExpressionMultiply
   inputs [
    A {from=MaterialExpressionTime_2}
    B {from=""}
   ]
   properties {constB=-0.1}
  }
  MaterialExpressionNamedRerouteDeclaration_0 {
   class=MaterialExpressionNamedRerouteDeclaration
   inputs [None {from=MaterialExpressionMultiply_1}]
   properties {name=PortalMask, nodeColor=<rgba 0.3828 1 0 1>}
  }
  MaterialExpressionNamedRerouteDeclaration_1 {
   class=MaterialExpressionNamedRerouteDeclaration
   inputs [None {from=MaterialExpressionTextureSample_4, output=R}]
   properties {name=PortalChannel, nodeColor=<rgba 0.3359 0 1 1>}
  }
  MaterialExpressionNamedRerouteDeclaration_11 {
   class=MaterialExpressionNamedRerouteDeclaration
   inputs [None {from=MaterialExpressionVectorParameter_2, output=RGB}]
   properties {name=BasisX, nodeColor=<rgba 1 0 0.2578 1>}
  }
  MaterialExpressionNamedRerouteDeclaration_12 {
   class=MaterialExpressionNamedRerouteDeclaration
   inputs [None {from=MaterialExpressionVectorParameter_10, output=RGB}]
   properties {name=BasisY, nodeColor=<rgba 0.9922 1 0 1>}
  }
  MaterialExpressionNamedRerouteDeclaration_13 {
   class=MaterialExpressionNamedRerouteDeclaration
   inputs [None {from=MaterialExpressionVectorParameter_11, output=RGB}]
   properties {name=BasisZ, nodeColor=<rgba 0.9453 0 1 1>}
  }
  MaterialExpressionNamedRerouteDeclaration_2 {
   class=MaterialExpressionNamedRerouteDeclaration
   inputs [None {from=MaterialExpressionLinearInterpolate_1}]
   properties {name=PortalColor, nodeColor=<rgba 0.0962 0.6255 1 1>}
  }
  MaterialExpressionNamedRerouteDeclaration_3 {
   class=MaterialExpressionNamedRerouteDeclaration
   inputs [None {from=MaterialExpressionVertexInterpolator_0, output=PS}]
   properties {name=PS-LocalPos, nodeColor=<rgba 0.875 0 1 1>}
  }
  MaterialExpressionNamedRerouteDeclaration_4 {
   class=MaterialExpressionNamedRerouteDeclaration
   inputs [None {from=MaterialExpressionTextureSample_3, output=G}]
   properties {name=SwirlMask, nodeColor=<rgba 0 0.25 1 1>}
  }
  MaterialExpressionNamedRerouteDeclaration_5 {
   class=MaterialExpressionNamedRerouteDeclaration
   inputs [None {from=MaterialExpressionDDX_0}]
   properties {name=DDX, nodeColor=<rgba 1 0.8906 0 1>}
  }
  MaterialExpressionNamedRerouteDeclaration_6 {
   class=MaterialExpressionNamedRerouteDeclaration
   inputs [None {from=MaterialExpressionDDY_0}]
   properties {name=DDY, nodeColor=<rgba 0.5703 0 1 1>}
  }
  MaterialExpressionNamedRerouteUsage_0 {class=MaterialExpressionNamedRerouteUsage}
  MaterialExpressionNamedRerouteUsage_1 {class=MaterialExpressionNamedRerouteUsage}
  MaterialExpressionNamedRerouteUsage_10 {class=MaterialExpressionNamedRerouteUsage}
  MaterialExpressionNamedRerouteUsage_11 {class=MaterialExpressionNamedRerouteUsage}
  MaterialExpressionNamedRerouteUsage_2 {class=MaterialExpressionNamedRerouteUsage}
  MaterialExpressionNamedRerouteUsage_3 {class=MaterialExpressionNamedRerouteUsage}
  MaterialExpressionNamedRerouteUsage_4 {class=MaterialExpressionNamedRerouteUsage}
  MaterialExpressionNamedRerouteUsage_5 {class=MaterialExpressionNamedRerouteUsage}
  MaterialExpressionNamedRerouteUsage_6 {class=MaterialExpressionNamedRerouteUsage}
  MaterialExpressionNamedRerouteUsage_7 {class=MaterialExpressionNamedRerouteUsage}
  MaterialExpressionNamedRerouteUsage_8 {class=MaterialExpressionNamedRerouteUsage}
  MaterialExpressionNamedRerouteUsage_9 {class=MaterialExpressionNamedRerouteUsage}
  MaterialExpressionOneMinus_1 {class=MaterialExpressionOneMinus, inputs [None {from=MaterialExpressionDistance_0}]}
  MaterialExpressionOneMinus_2 {class=MaterialExpressionOneMinus, inputs [None {from=MaterialExpressionNamedRerouteUsage_3}]}
  MaterialExpressionSaturate_0 {class=MaterialExpressionSaturate, inputs [None {from=MaterialExpressionSubtract_2}]}
  MaterialExpressionScalarParameter_1 {class=MaterialExpressionScalarParameter, properties {defaultValue=0.15, parameterName=SpeedX}}
  MaterialExpressionScalarParameter_2 {class=MaterialExpressionScalarParameter, properties {parameterName=SpeedY}}
  MaterialExpressionSmoothStep_0 {
   class=MaterialExpressionSmoothStep
   inputs [
    Min {from="", output=""}
    Max {from="", output=""}
    Value {from=MaterialExpressionSaturate_0, output=""}
   ]
   properties {constMax=0.35, constMin=0.25}
  }
  MaterialExpressionSmoothStep_1 {
   class=MaterialExpressionSmoothStep
   inputs [
    Min {from="", output=""}
    Max {from="", output=""}
    Value {from=MaterialExpressionOneMinus_1, output=""}
   ]
   properties {constMax=0.85, constMin=0.8}
  }
  MaterialExpressionSphereMask_0 {
   class=MaterialExpressionSphereMask
   inputs [
    A {from=MaterialExpressionDivide_0, output=""}
    B {from=MaterialExpressionConstant2Vector_0, output=""}
    Radius {from="", output=""}
    Hardness {from="", output=""}
   ]
   properties {attenuationRadius=0.4, hardnessPercent=1}
  }
  MaterialExpressionStep_0 {
   class=MaterialExpressionStep
   inputs [
    Y {from="", output=""}
    X {from=MaterialExpressionSmoothStep_0, output=""}
   ]
   properties {constX=0.885, constY=0.6}
  }
  MaterialExpressionStep_1 {
   class=MaterialExpressionStep
   inputs [
    Y {from=MaterialExpressionConstant_0, output=""}
    X {from=MaterialExpressionComponentMask_3, output=""}
   ]
   properties {constX=0}
  }
  MaterialExpressionSubtract_0 {
   class=MaterialExpressionSubtract
   inputs [
    A {from=MaterialExpressionMultiply_5}
    B {from=MaterialExpressionMultiply_2}
   ]
  }
  MaterialExpressionSubtract_1 {
   class=MaterialExpressionSubtract
   inputs [
    A {from=MaterialExpressionDivide_0}
    B {from=MaterialExpressionConstant2Vector_0}
   ]
  }
  MaterialExpressionSubtract_2 {
   class=MaterialExpressionSubtract
   inputs [
    A {from=MaterialExpressionSubtract_0}
    B {from=MaterialExpressionMultiply_3}
   ]
   properties {constB=0}
  }
  MaterialExpressionSubtract_3 {
   class=MaterialExpressionSubtract
   inputs [
    A {from=MaterialExpressionComponentMask_1}
    B {from=MaterialExpressionConstant2Vector_4}
   ]
  }
  MaterialExpressionSubtract_5 {
   class=MaterialExpressionSubtract
   inputs [
    A {from=MaterialExpressionAdd_7}
    B {from=MaterialExpressionVectorParameter_0, output=RGB}
   ]
  }
  MaterialExpressionTextureCoordinate_1 {class=MaterialExpressionTextureCoordinate}
  MaterialExpressionTextureSample_0 {
   class=MaterialExpressionTextureSample
   inputs [
    UVs {from=MaterialExpressionMultiply_0}
    Tex {from=""}
    "DDX(UVs)" {from=MaterialExpressionNamedRerouteUsage_6}
    "DDY(UVs)" {from=MaterialExpressionNamedRerouteUsage_7}
    Apply View MipBias {from=""}
   ]
   properties {mipValueMode=TMVM_Derivative, samplerSource=SSM_Wrap_WorldGroupSettings, texture=/Game/Global/Textures/Default/Noise/3xPerlin_M}
  }
  MaterialExpressionTextureSample_1 {
   class=MaterialExpressionTextureSample
   inputs [
    UVs {from=MaterialExpressionAdd_1}
    Tex {from=""}
    "DDX(UVs)" {from=MaterialExpressionNamedRerouteUsage_8}
    "DDY(UVs)" {from=MaterialExpressionNamedRerouteUsage_9}
    Apply View MipBias {from=""}
   ]
   properties {mipValueMode=TMVM_Derivative, samplerSource=SSM_Wrap_WorldGroupSettings, texture=&3xPerlin_M}
  }
  MaterialExpressionTextureSample_2 {
   class=MaterialExpressionTextureSample
   inputs [
    UVs {from=MaterialExpressionAdd_2}
    Tex {from=""}
    "DDX(UVs)" {from=MaterialExpressionNamedRerouteUsage_10}
    "DDY(UVs)" {from=MaterialExpressionNamedRerouteUsage_11}
    Apply View MipBias {from=""}
   ]
   properties {mipValueMode=TMVM_Derivative, samplerSource=SSM_Wrap_WorldGroupSettings, texture=&3xPerlin_M}
  }
  MaterialExpressionTextureSample_3 {
   class=MaterialExpressionTextureSample
   inputs [
    UVs {from=MaterialExpressionMaterialFunctionCall_2, output=Rotated Values}
    Tex {from=""}
    Apply View MipBias {from=""}
   ]
   properties {
    samplerSource=SSM_Wrap_WorldGroupSettings
    samplerType=SAMPLERTYPE_Masks
    texture=/BRCosmetics/Animation/Props/Gadgets_ProngCling/Textures/T_ProngCling_Gadgets_TA
   }
  }
  MaterialExpressionTextureSample_4 {
   class=MaterialExpressionTextureSample
   inputs [
    UVs {from=""}
    Tex {from=""}
    Apply View MipBias {from=""}
   ]
   properties {
    samplerSource=SSM_Wrap_WorldGroupSettings
    samplerType=SAMPLERTYPE_Masks
    texture=&T_ProngCling_Gadgets_TA
   }
  }
  MaterialExpressionTime_0 {class=MaterialExpressionTime}
  MaterialExpressionTime_2 {class=MaterialExpressionTime}
  MaterialExpressionVectorParameter_0 {class=MaterialExpressionVectorParameter, properties {parameterName=PC_PortalCenterLoc, defaultValue=<rgba 0 0 0 1>}}
  MaterialExpressionVectorParameter_10 {class=MaterialExpressionVectorParameter, properties {group=PrismParticle, parameterName=PC_PivotBasisY, defaultValue=<rgba 0 1 0 0>}}
  MaterialExpressionVectorParameter_11 {class=MaterialExpressionVectorParameter, properties {group=PrismParticle, parameterName=PC_PivotBasisZ, defaultValue=<rgba 0 0 1 0>}}
  MaterialExpressionVectorParameter_2 {class=MaterialExpressionVectorParameter, properties {group=PrismParticle, parameterName=PC_PivotBasisX, defaultValue=<rgba 1 0 0 0>}}
  MaterialExpressionVertexInterpolator_0 {class=MaterialExpressionVertexInterpolator, inputs [VS {from=MaterialExpressionLocalPosition_0, output=XYZ}], properties {}}
  MaterialExpressionWorldPosition_2 {class=MaterialExpressionWorldPosition, inputs [], properties {worldPositionShaderOffset=WPT_CameraRelative}}
 }}
