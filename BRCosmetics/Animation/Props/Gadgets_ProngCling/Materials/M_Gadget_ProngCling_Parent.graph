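# Parent material for the ProngCling gadget prop, built on the 2023 anime character shader for skeletal meshes and mesh deformers.
# It chains the shared Pre, Calculations and Post functions, using material attributes and the Cloth shading model.
# MF_Gadget_ProngCling sits between Pre and Post to process the lit and shaded colors.
# That gadget function also supplies the Opacity Mask set on the final attributes.
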Material /BRCosmetics/Animation/Props/Gadgets_ProngCling/Materials/M_Gadget_ProngCling_Parent
graph {
 properties {
  bAutomaticallySetUsageInEditor=false
  bUseMaterialAttributes=true
  bUsedWithMeshDeformer=true
  bUsedWithSkeletalMesh=true
  pixelDepthOffsetMode=PDOM_Legacy
  shadingModel=MSM_Cloth
  usedShadingModels=Cloth
 }
 nodes [
  n1 = MaterialFunctionCall!(CustomUV) {materialFunction=/Game/Global/Materials/Character_Parent_2023/Anime_Parent_2023/MaterialFunctions/MFC_AnimeShader_2023_Pre}
  n2 = MaterialFunctionCall!(CustomLightVector, OverrideNormal_WS, STT=@n1.STT, Normal_Map=@n1.STT, DFL=@n1.STT, FX=@n1.STT) {materialFunction=/Game/Global/Materials/Character_Parent_2023/Anime_Parent_2023/MaterialFunctions/MFC_AnimeShader_2023_Calculations}
  n3 = MaterialFunctionCall!(Color_Lit_In=@n1.Color_Lit, Color_Shaded_In=@n1.Color_Lit) {materialFunction=&/MF_Gadget_ProngCling}
  SetMaterialAttributes!(
   MaterialAttributes=MaterialFunctionCall!(
    Shading_Gradient=@n2.Shading_Gradient
    Threshold_Gradient=@n2.Shading_Gradient
    Specular_Gradient=@n2.Shading_Gradient
    Sheen_Mask=@n2.Shading_Gradient
    Thickness=@n2.Shading_Gradient
    DFL_Mask=@n2.Shading_Gradient
    FX=@n2.Shading_Gradient
    Metallic_Gradient=@n2.Shading_Gradient
    RimLight_Gradient=@n2.Shading_Gradient
    NormalBuffer=@n2.Shading_Gradient
    --------
    CINE_LightSources=@n2.Shading_Gradient
    CINE_LightColor=@n2.Shading_Gradient
    ---------
    Color_Lit=@n3.Color_Lit
    Color_Shaded=@n3.Color_Lit
    SRM=@n1.SRM
    Emissive=@n1.SRM
    DFL_Color=@n1.SRM
    Specular_Color=@n1.SRM
    RimLight_Color=@n1.SRM
   ) {materialFunction=/Game/Global/Materials/Character_Parent_2023/Anime_Parent_2023/MaterialFunctions/MFC_AnimeShader_2023_Post}.Result
   Opacity Mask=@n3.OpacityMask
  ) {attributeSetTypes [#1]}
 ]
}
