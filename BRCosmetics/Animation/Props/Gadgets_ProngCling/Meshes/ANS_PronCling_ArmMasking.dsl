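; Anim notify state tick for the Prong Cling gadget that feeds portal position and orientation to the character material, apparently for arm masking.
; Each tick it searches the children of the animated mesh's attach parent for the skeletal mesh using the portal mesh asset.
; It writes that mesh's JNT_PortalSocket world location and up, right and forward axes into the animated mesh's dynamic material vector parameters.
; An overlay material is also fetched but not used.
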(fn Received_NotifyTick
  (params (MeshComp :type Skeletal%20Mesh%20Component%20Object%20Reference) (Animation :type Anim%20Sequence%20Base%20Object%20Reference) (FrameDeltaTime :type Float%20%28single-precision%29) (EventReference :type Anim%20Notify%20Event%20Reference%20Structure%20%28by%20ref%29))
  (bind _array_element (Utilities|Array|ForEachLoopwithBreak (Components|GetChildrenComponents (Transformation|GetAttachParent MeshComp) false))
    (:LoopBody
      (bind _asskeletal_mesh_component (Utilities|Casting|CastToSkeletalMeshComponent _array_element))
      (if (== (Variables|/Script/Engine.SkinnedMeshComponent|GetSkeletalMesh _asskeletal_mesh_component) (Variables|Default|GetPortalMeshAsset))
        (Utilities|FlowControl|Sequence
          (:then_0
            _)
          (:then_1
            _))
        (bind _returnvalue (Fortnite|GetDynamicMaterialInstance (Variables|GetMeshComp)))
        (bind _returnvalue_2 (Class|FortCosmeticMaterialBlueprintLibrary|GetorCreateDynamicMaterialSlotOverlayMaterial (Variables|GetMeshComp)))
        (Rendering|Material|SetVectorParameterValue _returnvalue "PC_PortalCenterLoc" (Math|Conversions|ToLinearColor(Vector) (.location (Transformation|GetSocketTransform _asskeletal_mesh_component "JNT_PortalSocket" "RTS_World"))) :!class /Script/Engine.MaterialInstanceDynamic)
        (Rendering|Material|SetVectorParameterValue _returnvalue "PC_PivotBasisX" (Math|Conversions|ToLinearColor(Vector) (Math|Vector|GetUpVector (.rotation (Transformation|GetSocketTransform _asskeletal_mesh_component "JNT_PortalSocket" "RTS_World")))) :!class /Script/Engine.MaterialInstanceDynamic)
        (Rendering|Material|SetVectorParameterValue _returnvalue "PC_PivotBasisY" (Math|Conversions|ToLinearColor(Vector) (Math|Vector|GetRightVector (.rotation (Transformation|GetSocketTransform _asskeletal_mesh_component "JNT_PortalSocket" "RTS_World")))) :!class /Script/Engine.MaterialInstanceDynamic)
        (Rendering|Material|SetVectorParameterValue _returnvalue "PC_PivotBasisZ" (Math|Conversions|ToLinearColor(Vector) (Math|Vector|GetForwardVector (.rotation (Transformation|GetSocketTransform _asskeletal_mesh_component "JNT_PortalSocket" "RTS_World")))) :!class /Script/Engine.MaterialInstanceDynamic)))
    (:Completed
      _))
  (return false))
