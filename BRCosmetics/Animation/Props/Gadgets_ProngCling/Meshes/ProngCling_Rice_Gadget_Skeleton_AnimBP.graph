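# Animation Blueprint for the ProngCling Rice gadget prop that keeps arm materials aligned to a portal socket; its pose is only the DefaultSlot montage slot.
# Shortly after begin play, it takes the portal mesh and the arm mesh from the Fort pawn's emote mesh data and caches the arm's dynamic and overlay materials.
# After each animation evaluation, it feeds JNT_PortalSocket's world location and axes to both materials, with up as X and forward as Z.
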AnimBlueprint /BRCosmetics/Animation/Props/Gadgets_ProngCling/Meshes/ProngCling_Rice_Gadget_Skeleton_AnimBP
target_skeleton=&/ProngCling_Rice_Gadget_Skeleton
graphs [
 AnimGraph {nodes [Output Pose(Result=Slot 'DefaultSlot'())]}
 EventGraph {nodes [
   Event BlueprintBeginPlay() {is=title K2Node_Event}
   > Parent: BlueprintBeginPlay() {is=title K2Node_CallParentFunction}
   > Delay(Duration=0.1)
   > n1 = Cast To FortPlayerPawn(Object=TryGetPawnOwner())
   > n2 = Values!(
    execute
    TargetMap:Map of Integers to Emote Mesh Comp Data Structures=Get EmoteMeshCompDatas!(self:Fort Pawn Object Reference=@n1) -> (EmoteMeshCompDatas:Map of Integers to Emote Mesh Comp Data Structures)
   ) -> (then, Values:Array of Emote Mesh Comp Data Structures)
   > Set Portal Mesh(Portal Mesh:Mesh Component Object Reference=Break Emote Mesh Comp Data!(EmoteMeshCompData:"Emote Mesh Comp Data Structure (by ref)"="Get (a ref)"(Array:Array of Emote Mesh Comp Data Structures=@n2, Dimension 1=0) -> (Output:"Emote Mesh Comp Data Structure (by ref)")) -> (MeshComp:Mesh Component Object Reference, PropExclusionTag:Gameplay Tag Structure, bTickEvenWhenNotRendered:bool=~) {is=title K2Node_BreakStruct}) -> (Output_Get:Mesh Component Object Reference)
   n3 = Break Emote Mesh Comp Data(EmoteMeshCompData="Get (a ref)"(Array=@n2, Dimension 1=1)) -> (bTickEvenWhenNotRendered=~)
   > Set Arm MID(Arm MID:Material Instance Dynamic Object Reference=GetDynamicMaterialInstance!(MeshComponent:Mesh Component Object Reference=@n3, MaterialIndex:int=0) -> (Material Instance Dynamic:Material Instance Dynamic Object Reference)) -> (Output_Get:Material Instance Dynamic Object Reference)
   > n4 = GetOrCreateDynamicMaterialSlotOverlayMaterial!(execute, MeshComponent:Mesh Component Object Reference=@n3, Index:int=0, MaterialName:Name=None) -> (then, ReturnValue:Material Instance Dynamic Object Reference)
   > Set Arm OverlayMID(Arm OverlayMID:Material Instance Dynamic Object Reference=@n4) -> (Output_Get:Material Instance Dynamic Object Reference)
   Event BlueprintPostEvaluateAnimation() {is=title K2Node_Event}
   > Parent: BlueprintPostEvaluateAnimation() {is=title K2Node_CallParentFunction}
   n5 = Get Arm MID() -> (Arm MID:Material Instance Dynamic Object Reference)
   > Is Valid(InputObject=@n5)
   n6 = BreakTransform(InTransform=GetSocketTransform(self=Get Portal Mesh() -> (Portal Mesh:Mesh Component Object Reference), InSocketName=JNT_PortalSocket, TransformSpace=RTS_World))
   n7 = Get Arm OverlayMID() -> (Arm OverlayMID:Material Instance Dynamic Object Reference)
   > SetVectorParameterValue(
    self=[@n7, @n5]
    ParameterName=PC_PortalCenterLoc
    Value="To LinearColor (Vector)"!(InVec:Vector=@n6) -> (ReturnValue:Linear Color Structure)
   )
   > SetVectorParameterValue(
    self=[@n7, @n5]
    ParameterName=PC_PivotBasisX
    Value="To LinearColor (Vector)"(InVec=GetUpVector!(InRot:Rotator=@n6.Rotation) -> (ReturnValue:Vector))
   )
   > SetVectorParameterValue(
    self=[@n7, @n5]
    ParameterName=PC_PivotBasisY
    Value="To LinearColor (Vector)"(InVec=GetRightVector!(InRot:Rotator=@n6.Rotation) -> (ReturnValue:Vector))
   )
   > SetVectorParameterValue(
    self=[@n7, @n5]
    ParameterName=PC_PivotBasisZ
    Value="To LinearColor (Vector)"(InVec=GetForwardVector!(InRot:Rotator=@n6.Rotation) -> (ReturnValue:Vector))
   )
  ]}
]
variables [
 Arm MID {type=MaterialInstanceDynamic}
 Arm OverlayMID {type=MaterialInstanceDynamic}
 Portal Mesh {type=MeshComponent, default=null}
]
